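# Conversation database for the TameFoot NPC, defining what it says when a player starts talking to it.
# The entry point leads to one speech node with a default greeting, a first-ever-conversation line and lines for players wearing specific character skins.
# Skin lines use fixed priority and weight overrides, while the first-meeting line takes its priority from a curve row.
# The speech node then links to the shared NPC quest and basic-services conversations.
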ConversationDatabase /TameFoot/NPCs/TameFoot/NPCLibrary_Conversation_TameFoot
entry_points [{tag=Conversation.Entry.TameFoot, destinations [#1]}]
nodes [
 ConversationEntryPointNode_7 {
  id=#1
  class=ConversationEntryPointNode
  outputs [#2]
  properties {entryTag {tagName=Conversation.Entry.TameFoot}}
 }
 Speech_C_4 {
  id=#2
  class=Speech_C
  outputs [#3, #4]
  properties {
   nodeName=""
   generalConfig {
    defaultMessage=Something to report?
    contextualMessages [
     {
      message=The Seven owes you a debt.
      contextRequirements [{requirement {class=FortControllerRequirement_IsFirstEverConversationWithNPC}}]
      priority {curve {rowName=Tandem.Speech.Priority.FirstEverInteraction}}
     }
     {
      message="We appreciate the Peace Syndicate's help in our cleanup efforts. "
      contextRequirements [{requirement {
         class=FortControllerRequirement_HasCID
         properties {allowedCIds [&CID_A_425_Athena_Commando_F_BlueJay]}
        }}]
      priority {value=10, curve {curveTable=null, rowName=None}}
      weight {value=10}
     }
     {
      message=I haven't had time to read your latest study.
      contextRequirements [{requirement {
         class=FortControllerRequirement_HasCID
         properties {allowedCIds [/BRCosmetics/Athena/Items/Cosmetics/Characters/CID_540_Athena_Commando_M_MeteorManRemix]}
        }}]
      priority {value=10, curve {curveTable=null, rowName=None}}
      weight {value=10}
     }
     {
      message=We've defeated The Imagined Order. Take a breath.
      contextRequirements [{requirement {
         class=FortControllerRequirement_HasCID
         properties {allowedCIds [/BRCosmetics/Athena/Items/Cosmetics/Characters/CID_A_373_Athena_Commando_M_OriginPrisoner]}
        }}]
      priority {value=10, curve {curveTable=null, rowName=None}}
      weight {value=10}
     }
     {
      message=You want to stay? Permanently? Let me think about it.
      contextRequirements [{requirement {
         class=FortControllerRequirement_HasCID
         properties {allowedCIds [/BRCosmetics/Athena/Items/Cosmetics/Characters/CID_140_Athena_Commando_M_Visitor]}
        }}]
      priority {value=10, curve {curveTable=null, rowName=None}}
      weight {value=10}
     }
     {
      message=Good flying out there.
      contextRequirements [{requirement {
         class=FortControllerRequirement_HasCID
         properties {allowedCIds [/BRCosmetics/Athena/Items/Cosmetics/Characters/CID_547_Athena_Commando_F_Meteorwoman]}
        }}]
      priority {value=10, curve {curveTable=null, rowName=None}}
      weight {value=10}
     }
     {
      message=You'd be giving me gray hair if I had any left.
      contextRequirements [{requirement {
         class=FortControllerRequirement_HasCID
         properties {allowedCIds [/BRCosmetics/Athena/Items/Cosmetics/Characters/CID_A_374_Athena_Commando_F_Binary]}
        }}]
      priority {value=10, curve {curveTable=null, rowName=None}}
      weight {value=10}
     }
     {
      message="You can relax now. "
      contextRequirements [{requirement {
         class=FortControllerRequirement_HasCID
         properties {allowedCIds [/BRCosmetics/Athena/Items/Cosmetics/Characters/CID_A_395_Athena_Commando_F_BinaryTwin]}
        }}]
      priority {value=10, curve {curveTable=null, rowName=None}}
      weight {value=10}
     }
    ]
   }
  }
 }
 ConversationLinkNode_0 {id=#4, class=ConversationLinkNode, properties {remoteEntryTag {tagName=Conversation.Entry.BasicServices}}}
 ConversationLinkNode_1 {id=#3, class=ConversationLinkNode, properties {remoteEntryTag {tagName=Conversation.Entry.NPCLibrary.Quests}}}
]
